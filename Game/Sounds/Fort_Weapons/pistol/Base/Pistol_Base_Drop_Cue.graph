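# Sound cue for the base pistol drop sound, routed through the Fort_Loot_Collect sound class.
# A single wave player feeds a modulator that randomizes volume within a reduced range, further scaled by the cue's volume multiplier.
# Weapon drop attenuation handles spatial falloff, and the shared Weapon_Drop_StopOldest concurrency set is assigned to limit overlapping drops.
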SoundCue /Game/Sounds/Fort_Weapons/pistol/Base/Pistol_Base_Drop_Cue
root=Modulator_1
nodes [
 {id=Modulator_1, children [WavePlayer_0], properties {volumeMax=0.6, volumeMin=0.5}}
 {id=WavePlayer_0, properties {soundWaveAssetPtr=&/Pistol_Base_Drop_01}}
]
properties {
 attenuationSettings=&Fort_Weapon_Drop_Att
 maxDistance=3000
 soundClassObject=&Fort_Loot_Collect
 volumeMultiplier=0.6
 concurrencyOverrides {maxCount=1, resolutionRule=PreventNew}
 concurrencySet [&Weapon_Drop_StopOldest]
}
