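# Sound cue for the base pistol's gunfire, apparently as heard from third person, layered by listening distance.
# A mixer blends close, medium and distant layers at descending levels, each with its own attenuation asset, slight volume modulation and randomly chosen wave variant.
# Sound class overrides route close and medium to the third-person fire class and distant to a distant-fire class.
# Local and global gunfire concurrency sets are assigned to limit stacking.
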SoundCue /Game/Sounds/Fort_Weapons/pistol/Base/Pistol_Base_Fire_Cue
root=Mixer_2
nodes [
 {id=Attenuation_3, children [Modulator_0], properties {attenuationSettings=&Fort_Weapon_Projectile_Close_Athena_Att}}
 {
  id=Attenuation_6
  children [Modulator_1]
  properties {attenuationSettings=&Fort_Weapon_Projectile_Medium_Mono_Athena_Att}
 }
 {
  id=Attenuation_7
  children [Modulator_2]
  properties {attenuationSettings=&Fort_Weapon_Projectile_Distant_Mono_Athena_Att}
 }
 {id=Mixer_2, children [SoundClass_3, SoundClass_10, SoundClass_9], properties {inputVolume [1,0.6,0.1]}}
 {id=Modulator_0, children [Random_2], properties {volumeMax=1.4, volumeMin=1.3}}
 {id=Modulator_1, children [Random_3], properties {volumeMax=2.4, volumeMin=2.3}}
 {id=Modulator_2, children [Random_4], properties {volumeMax=2.3, volumeMin=2.2}}
 {id=Random_2, children [WavePlayer_5,WavePlayer_6,WavePlayer_7,WavePlayer_8], properties {weights [1,1,1,1]}}
 {id=Random_3, children [WavePlayer_9, WavePlayer_10, WavePlayer_11], properties {weights [1,1,1]}}
 {id=Random_4, children [WavePlayer_13, WavePlayer_14, WavePlayer_15], properties {weights [1,1,1]}}
 {
  id=SoundClass_10
  children [Attenuation_6]
  properties {soundClassOverride=&Fort_WeaponRanged_Fire_3P_PSM}
 }
 {
  id=SoundClass_3
  children [Attenuation_3]
  properties {soundClassOverride=&Fort_WeaponRanged_Fire_3P_PSM}
 }
 {
  id=SoundClass_9
  children [Attenuation_7]
  properties {soundClassOverride=&Fort_WeaponRanged_Fire_Distant}
 }
 {id=WavePlayer_10, properties {soundWaveAssetPtr=&/Pistol_Base_Fire_Medium_02}}
 {id=WavePlayer_11, properties {soundWaveAssetPtr=&/Pistol_Base_Fire_Medium_03}}
 {id=WavePlayer_13, properties {soundWaveAssetPtr=&/Pistol_Base_Fire_Distant_01}}
 {id=WavePlayer_14, properties {soundWaveAssetPtr=&/Pistol_Base_Fire_Distant_02}}
 {id=WavePlayer_15, properties {soundWaveAssetPtr=&/Pistol_Base_Fire_Distant_03}}
 {id=WavePlayer_5, properties {soundWaveAssetPtr=&/Pistol_Base_Fire_01}}
 {id=WavePlayer_6, properties {soundWaveAssetPtr=&/Pistol_Base_Fire_02}}
 {id=WavePlayer_7, properties {soundWaveAssetPtr=&/Pistol_Base_Fire_03}}
 {id=WavePlayer_8, properties {soundWaveAssetPtr=&/Pistol_Base_Fire_04}}
 {id=WavePlayer_9, properties {soundWaveAssetPtr=&/Pistol_Base_Fire_Medium_01}}
]
properties {
 maxDistance=27500
 soundClassObject=&Fort_WeaponRanged_Fire_3P_PSM
 volumeMultiplier=0.5
 attenuationOverrides {falloffDistance=4100, attenuationShapeExtents {x=900}}
 concurrencyOverrides {bLimitToOwner=true, maxCount=2, resolutionRule=StopOldest}
 concurrencySet [&Weapon_Gunfire_Local_Conc, &Weapon_Gunfire_Global_Conc]
}
