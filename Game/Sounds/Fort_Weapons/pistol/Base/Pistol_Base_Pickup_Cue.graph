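# Sound cue for picking up the base pistol, routed to the loot-collect sound class.
# A single wave player feeds a Modulator that slightly randomizes volume on each play, with reduced overall cue volume.
# It uses shared weapon-pickup attenuation for distance falloff and the PickupMaximum_Oldest concurrency set to limit overlapping pickup sounds.
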SoundCue /Game/Sounds/Fort_Weapons/pistol/Base/Pistol_Base_Pickup_Cue
root=Modulator_0
nodes [
 {id=Modulator_0, children [WavePlayer_4], properties {volumeMax=0.95, volumeMin=0.85}}
 {id=WavePlayer_4, properties {soundWaveAssetPtr=&/Pistol_Base_Pickup_01}}
]
properties {
 attenuationSettings=&Fort_Weapon_Pickup_Att
 maxDistance=3500
 soundClassObject=&Fort_Loot_Collect
 volumeMultiplier=0.4615
 concurrencyOverrides {maxCount=1, resolutionRule=PreventNew}
 concurrencySet [&PickupMaximum_Oldest]
}
