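# Sound cue for the base pistol's reload insert sound, classed as weapon foley.
# A single wave player feeds a Modulator root, reusing a CandyCorn LMG reload insert sample rather than a pistol-specific one.
# The cue raises that sample's pitch and lowers its volume, with spatial falloff from the shared weapon reload attenuation asset.
# Its own concurrency override prevents new instances from starting once the limit is reached.
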SoundCue /Game/Sounds/Fort_Weapons/pistol/Base/Pistol_Base_ReloadInsert_Cue
root=Modulator_0
nodes [
 {id=Modulator_0, children [WavePlayer_0]}
 {id=WavePlayer_0, properties {soundWaveAssetPtr=/Game/Sounds/Fort_Weapons/AutoRifle/CandyCornLMG/CandyCorn_LMG_ReloadInsertB_01}}
]
properties {
 attenuationSettings=&Fort_Weapon_Reload_Att
 bOverrideConcurrency=true
 maxDistance=2250
 pitchMultiplier=1.2
 soundClassObject=&Fort_WeaponFoley
 volumeMultiplier=0.351
 attenuationOverrides {falloffDistance=4100, attenuationShapeExtents {x=900}}
 concurrencyOverrides {resolutionRule=PreventNew}
}
